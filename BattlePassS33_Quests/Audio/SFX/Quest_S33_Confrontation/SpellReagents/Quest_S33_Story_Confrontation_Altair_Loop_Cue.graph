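# Looping sound cue for the Altair in the Season 33 Confrontation story quest's spell reagents audio.
# A single looping wave feeds a one-entry random node, then a modulator for pitch variation and a mixer.
# It uses the chest attenuation settings and the gadgets sound class.
# Overridden concurrency allows one instance per owner and prevents new ones from starting.
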SoundCue /BattlePassS33_Quests/Audio/SFX/Quest_S33_Confrontation/SpellReagents/Quest_S33_Story_Confrontation_Altair_Loop_Cue
root=Mixer_0
nodes [
 {id=Mixer_0, children [Modulator_0], properties {inputVolume [1]}}
 {id=Modulator_0, children [Random_0], properties {pitchMax=1.02}}
 {id=Random_0, children [WavePlayer_0], properties {weights [1]}}
 {
  id=WavePlayer_0
  properties {
   bLooping=true
   soundWaveAssetPtr=&/Quest_S33_Story_Confrontation_Altair_Loop_01
  }
 }
]
properties {
 attenuationSettings=/Game/Sounds/Attenuation/Fort_Chests_Att
 bOverrideConcurrency=true
 maxDistance=1300
 soundClassObject=&Fort_Gadgets
 volumeMultiplier=1
 concurrencyOverrides {bLimitToOwner=true, maxCount=1, resolutionRule=PreventNew}
}
