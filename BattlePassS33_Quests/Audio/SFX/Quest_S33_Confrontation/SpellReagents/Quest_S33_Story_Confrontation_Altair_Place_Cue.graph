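# Sound cue for a Season 33 Confrontation story quest, named for placing an Altair spell reagent.
# A single wave, Altair_Place_02, feeds a one-entry Random node into a Mixer whose second input is empty.
# It borrows a supply-drop ambient loop attenuation, boosts volume and uses the non-critical in-game UI sound class.
# A concurrency override allows one instance per owner and blocks new plays while it runs.
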SoundCue /BattlePassS33_Quests/Audio/SFX/Quest_S33_Confrontation/SpellReagents/Quest_S33_Story_Confrontation_Altair_Place_Cue
root=Mixer_0
nodes [
 {id=Mixer_0, children [Random_0,null], properties {inputVolume [1,1]}}
 {id=Random_0, children [WavePlayer_2], properties {weights [1]}}
 {
  id=WavePlayer_2
  properties {soundWaveAssetPtr=&/Quest_S33_Story_Confrontation_Altair_Place_02}
 }
]
properties {
 attenuationSettings=/Game/Athena/Sounds/SupplyDrops/Fort_HiddenStar_Ambient_Loop_Att
 bOverrideConcurrency=true
 maxDistance=2500
 soundClassObject=&Fort_UI_InGame_NonCritical
 volumeMultiplier=3
 concurrencyOverrides {bLimitToOwner=true, maxCount=1, resolutionRule=PreventNew}
}
